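; Base Blueprint for a Sprite worn as a backpack, attaching it to a Fort player pawn's Backpack_BR socket.
; On initialization it sets the owning pawn, choosing a new or legacy path via the Fort.Rollback.SpritesSyncLoadFix CVar.
; Both attach once the body mesh is loaded, polling on a looping timer until then, which is cleared on EndPlay.
; Montage, hidden-change and BackpackAttached events are empty hooks, the latter called after attachment.
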(event EventEndPlay
  (params (EndPlayReason :type EEndPlayReason%20Enum :default "Destroyed"))
  (|CallParentFunction EndPlayReason)
  (Utilities|Time|ClearandInvalidateTimerbyHandle (Variables|Default|GetWaitForPart))
  (Variables|Default|SetWaitForPart))

(event Custom|InitializeForPawn
  (params (FortPlayerPawn :type Fort%20Player%20Pawn%20Object%20Reference))
  (if (> (Fortnite|GetNumericCVar "Fort.Rollback.SpritesSyncLoadFix") 0)
    (CallFunction|InitializeNewCode FortPlayerPawn)
    (else
      (CallFunction|InitializeOldCode FortPlayerPawn))))

(event Custom|Initialize_NewCode
  (params (FortPawn :type Fort%20Player%20Pawn%20Object%20Reference))
  (CallFunction|SetOwningPawn FortPawn)
  (if (CallFunction|IsBodyPartLoaded)
    (CallFunction|AttachtoPawn "Backpack_BR")
    (CallFunction|BackpackAttached FortPawn)
    (else
      (bind _returnvalue (Utilities|Time|SetTimerbyEvent (delegate Custom|CustomEvent) 0.2 true true))
      (Variables|Default|SetWaitForPart _returnvalue))))

(event Custom|CustomEvent
  (if (CallFunction|IsBodyPartLoaded)
    (Utilities|Time|ClearandInvalidateTimerbyHandle (Variables|Default|GetWaitForPart))
    (Variables|Default|SetWaitForPart)
    (CallFunction|AttachtoPawn "Backpack_BR")
    (CallFunction|BackpackAttached (stale Custom|Initialize_NewCode FortPawn))))

(event Custom|Initialize_OldCode
  (params (FortPawn :type Fort%20Player%20Pawn%20Object%20Reference))
  (bind _returnvalue (Customization|GetSkeletalMeshforPartType FortPawn "Body" :!class /Script/FortniteGame.FortPlayerPawn))
  (CallFunction|SetOwningPawn FortPawn)
  (Utilities|IsValid _returnvalue
    (:Is%20Valid
      (Transformation|AttachActorToComponent _returnvalue "Backpack_BR" "SnapToTarget" "SnapToTarget" "KeepWorld")
      (CallFunction|BackpackAttached FortPawn))
    (:Is%20Not%20Valid
      (bind _returnvalue_2 (Utilities|Time|SetTimerbyEvent (delegate Custom|CheckIfLoaded) 0.2 true true))
      (Variables|Default|SetWaitForPart _returnvalue_2))))

(event Custom|CheckIfLoaded
  (Utilities|IsValid (Customization|GetSkeletalMeshforPartType (stale Custom|Initialize_OldCode FortPawn) "Body" :!class /Script/FortniteGame.FortPlayerPawn)
    (:Is%20Valid
      (Utilities|Time|ClearandInvalidateTimerbyHandle (Variables|Default|GetWaitForPart))
      (Variables|Default|SetWaitForPart)
      (Transformation|AttachActorToComponent (Customization|GetSkeletalMeshforPartType (stale Custom|Initialize_OldCode FortPawn) "Body" :!class /Script/FortniteGame.FortPlayerPawn) "Backpack_BR" "SnapToTarget" "SnapToTarget" "KeepWorld")
      (CallFunction|BackpackAttached (stale Custom|Initialize_OldCode FortPawn)))))

(event Custom|PlayAgitatedMontage)

(event Custom|PlayCelebrateMontage)

(event Custom|PlayCrouchMontage)

(event Custom|PlayCryMontage)

(event Custom|BackpackAttached
  (params (PlayerPawn :type Fort%20Player%20Pawn%20Object%20Reference)))

(event Custom|OnHiddenChanged
  (params (NewHidden :type Boolean)))

(fn ConstructionScript)
