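; Parent weapon Blueprint for held Sprite items, managing their effects, music tags and holster state.
; Equipping adds music tags to the instigator and equips the Sprite through SpritesUtilityLibrary; unequipping removes both and clears the holster flag.
; An async tag-count watcher on the force-backpack-hidden tag toggles equip-animation suppression and holsters the weapon when the tag is present.
; Visibility changes send a gameplay event to the owner, with visibility as the magnitude.
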(event FortWeapon|EventOnInstigatorSet
  (|CallParentFunction))

(event FortWeapon|EventOnEquippedAndAttached
  (bind _returnvalue (Game|GetInstigator))
  (|CallParentFunction)
  (CallFunction|RegisterAudioTag true)
  (Class|SpritesUtilityLibrary|EquipSprite (CallFunction|GetPlayerController) (Variables|Default|GetEquipped%20Effect) (Variables|Default|GetEnv%20Sprites%20Effect))
  (bind _returnvalue_2 (GameplayTags|HasMatchingGameplayTag(Message) _returnvalue (Variables|Default|GetForceBackpackHiddenTag)))
  (CallFunction|SetDisableEquipAnimation _returnvalue_2)
  (bind (_asynctaskproxy _tagcount) (Ability|Tasks|WaitGameplayTagCountChangedOnActor _returnvalue (Variables|Default|GetForceBackpackHiddenTag))
    (:TagCountChanged
      (CallFunction|SetDisableEquipAnimation (> _tagcount 0))
      (if (> _tagcount 0)
        (bind _asfort_pawn (Utilities|Casting|CastToFortPawn (Game|GetInstigator)))
        (Weapon|SetHolsterWeapon _asfort_pawn self true false (Variables|Default|GetHolster%20Id))))))

(event FortWeapon|EventOnUnEquippedAndDetached
  (|CallParentFunction)
  (CallFunction|RegisterAudioTag false)
  (Class|SpritesUtilityLibrary|UnequipSprite (CallFunction|GetPlayerController) (Variables|Default|GetEffect%20Tag%20to%20Remove))
  (bind _asfort_pawn (Utilities|Casting|CastToFortPawn (Game|GetInstigator)))
  (Weapon|SetHolsterWeapon _asfort_pawn self false false "Sprite"))

(event EventOnWeaponVisibilityChanged
  (params (bVisible :type Boolean) (bSetForLocalControllerOnly :type Boolean))
  (|CallParentFunction bVisible bSetForLocalControllerOnly)
  (Ability|SendGameplayEventtoActor (Actor|GetOwner) "(TagName=\"Athena.Player.Sprite.ItemVisibilityChanged\")" (Utilities|Struct|MakeGameplayEventData "(TagName=\"\")" self 0 0 0 "()" "(GameplayTags=)" "(GameplayTags=)" (Math|Conversions|ToFloat(Boolean) bVisible))))

(fn RegisterAudioTag
  (params (bRegister :type Boolean))
  (bind _returnvalue (Game|GetInstigator))
  (bind _musictags (Variables|Music|GetMusicTags))
  ; Handles adding required music tag to instigator actor
  (Utilities|IsValid _returnvalue
    (:Is%20Valid
      (if bRegister
        (Ability|GameplayEffect|AddLooseGameplayTags _returnvalue _musictags true)
        (else
          (Ability|GameplayEffect|RemoveLooseGameplayTags _returnvalue _musictags true))))))

(fn ConstructionScript)
